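# Sound cue for throwing a candy consumable, played as a gameplay cue sound for items in Athena.
# A single wave player feeds the root with the generic consumables item sound at reduced volume, with no randomization or modulation.
# It routes through the gadgets item sound class and overrides the gadget-placement attenuation with its own falloff, shape extent and stereo spread.
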SoundCue /FNGameplayCues/Athena/Sounds/Items/Consumable_Candy_Throw_Cue
root=WavePlayer_0
nodes [{id=WavePlayer_0, properties {soundWaveAssetPtr=&/Item_Consumables_Generic}}]
properties {
 attenuationSettings=/Game/Sounds/Attenuation/Fort_GadgetPlace_Att
 bOverrideAttenuation=true
 maxDistance=2500
 soundClassObject=&Fort_Gadgets_Items
 volumeMultiplier=0.5
 attenuationOverrides {falloffDistance=1600, stereoSpread=2000, attenuationShapeExtents {x=900}}
}
